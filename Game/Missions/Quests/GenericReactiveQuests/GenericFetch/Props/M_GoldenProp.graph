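# M_GoldenProp is a simple Default Lit material in the generic reactive fetch quest's props folder, apparently for a gold prop look.
# It holds a warm gold Color vector parameter, a low shine roughness scalar parameter and a constant of one, with no textures.
# The listing shows no connections from these nodes to material outputs.
# Usage flags cover skeletal meshes, mesh particles, mesh deformers and Nanite.
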Material /Game/Missions/Quests/GenericReactiveQuests/GenericFetch/Props/M_GoldenProp
graph {
 properties {
  bUsedWithMeshDeformer=true
  bUsedWithMeshParticles=true
  bUsedWithNanite=true
  bUsedWithSkeletalMesh=true
  pixelDepthOffsetMode=PDOM_Legacy
  usedShadingModels=Default Lit
 }
 nodes [
  Constant() {r=1}
  ScalarParameter() {defaultValue=0.28, parameterName=Shine Roughnmess}
  VectorParameter() {parameterName=Color, defaultValue=<rgba 1 0.5683 0.2174 1>}
 ]
}
